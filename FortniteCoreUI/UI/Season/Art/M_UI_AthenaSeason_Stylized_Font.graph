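# Translucent UI material for Athena season stylized text, producing a two-color radial gradient with adjustable opacity.
# An exponential radial gradient with parameterized center, radius and density blends both the colors and the falloff-to-hotspot alpha.
# UseVGrad blends in a vertical UV fade, applied through a DotProduct of scalars as a multiply, before OverallAlpha scales the result.
# A static switch, UseIconMask, optionally masks that alpha with a texture's alpha channel, defaulting to a wordmark logo.
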Material /FortniteCoreUI/UI/Season/Art/M_UI_AthenaSeason_Stylized_Font
graph {
 properties {
  bUsedWithStaticMesh=false
  blendMode=BLEND_Translucent
  materialDomain=MD_UI
  pixelDepthOffsetMode=PDOM_Legacy
  usedShadingModels=Default Lit
 }
 nodes [
  n1 = MaterialFunctionCall(
   UVs=TextureCoordinate()
   CenterPosition=AppendVector(A=ScalarParameter() {parameterName=U_Position}, B=ScalarParameter() {parameterName=V_Position})
   Radius=ScalarParameter() {defaultValue=1.5, parameterName=Radius}
   Density=ScalarParameter() {defaultValue=1.666, parameterName=Density}
  ) {materialFunction=&RadialGradientExponential}
  LinearInterpolate(
   A=VectorParameter() {parameterName=Color2, defaultValue=<rgba 0 0.0194 0.3185 1>}.RGB
   B=VectorParameter() {parameterName=Color1, defaultValue=<rgba 0 0.0578 0.6038 1>}.RGB
   Alpha=@n1.RadialGradientExponential
  )
  n2 = LinearInterpolate(
   A=ScalarParameter() {defaultValue=0.75, parameterName=FallOffAlpha}
   B=ScalarParameter() {defaultValue=1, parameterName=HotSpotAlpha}
   Alpha=@n1.RadialGradientExponential
  )
  n3 = ScalarParameter() {parameterName=V_Grad_FallOffAlpha}
  n4 = ScalarParameter() {defaultValue=1, parameterName=V_Grad_Hotspot}
  n5 = OneMinus(None=ComponentMask(None=TextureCoordinate()) {g=true})
  n6 = ScalarParameter() {parameterName=MoveFallOffPoint}
  n7 = Saturate(None=Multiply(A=Add(A=@n5, B=@n6), B=ScalarParameter() {defaultValue=1, parameterName=V_Grad_FallOff_Start}))
  n8 = LinearInterpolate(A=@n2, B=DotProduct(A=@n2, B=LinearInterpolate(A=@n3, B=@n4, Alpha=@n7)), Alpha=ScalarParameter() {parameterName=UseVGrad})
  n9 = Saturate(None=Multiply(A=Saturate(None=@n8), B=ScalarParameter() {defaultValue=1, parameterName=OverallAlpha}))
  StaticSwitchParameter(
   True=LinearInterpolate(
    B=@n9
    Alpha=TextureSampleParameter2D() {parameterName=TextureMask, texture=/FortniteCoreUI/Foundation/Textures/Logo/T-Logo-Wordmark-1024x256}.A
   )
   False=@n9
  ) {parameterName=UseIconMask}
 ]
}
